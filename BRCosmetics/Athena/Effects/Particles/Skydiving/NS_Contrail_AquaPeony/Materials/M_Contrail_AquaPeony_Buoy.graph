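# Default Lit material for Niagara mesh particles, shading the buoy in the AquaPeony skydiving contrail cosmetic.
# A packed mask texture's blue channel selects base versus stripe color, and its green channel is overlay-blended on top.
# Alpha, scaled by the grunge color's alpha, blends in grunge tint and also interpolates between min and max roughness for base and stripe.
# A separate normal map texture supplies surface detail.
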Material /BRCosmetics/Athena/Effects/Particles/Skydiving/NS_Contrail_AquaPeony/Materials/M_Contrail_AquaPeony_Buoy
graph {
 properties {bUsedWithNiagaraMeshParticles=true, pixelDepthOffsetMode=PDOM_Legacy, usedShadingModels=Default Lit}
 nodes [
  n1 = TextureSample() {
   samplerType=SAMPLERTYPE_LinearColor
   texture=/BRCosmetics/Athena/Effects/Particles/Skydiving/NS_Contrail_AquaPeony/Textures/T_Contrail_AquaPeony_Buoy_M
  }
  LinearInterpolate(
   A=LinearInterpolate(
    A=ScalarParameter() {defaultValue=0.5, parameterName=Roughness Base Min}
    B=ScalarParameter() {defaultValue=0.8, parameterName=Roughness Base Max}
    Alpha=@n1.A
   )
   B=LinearInterpolate(
    A=ScalarParameter() {defaultValue=0.8, parameterName=Roughness Stripe Min}
    B=ScalarParameter() {defaultValue=1, parameterName=Roughness Stripe Max}
    Alpha=@n1.A
   )
   Alpha=@n1.B
  )
  n2 = VectorParameter() {parameterName=Color Base, defaultValue=<rgba 0.45 0.0519 0.045 1>}
  n3 = VectorParameter() {parameterName=Color Stripe, defaultValue=<rgba 0.8 0.7804 0.64 1>}
  n4 = VectorParameter() {parameterName=Color Grunge, defaultValue=<rgba 0.045 0.2203 0.0466 0.4>}
  LinearInterpolate(
   A=MaterialFunctionCall(Base=LinearInterpolate(A=@n2.RGB, B=@n3.RGB, Alpha=@n1.B), Blend=@n1.G) {materialFunction=&Blend_Overlay}.Result
   B=@n4.RGB
   Alpha=Multiply(A=@n1.A, B=@n4.A)
  )
  Power(Exp=@n4.A)
  TextureSample() {
   samplerType=SAMPLERTYPE_Normal
   texture=/BRCosmetics/Athena/Effects/Particles/Skydiving/NS_Contrail_AquaPeony/Textures/T_Contrail_AquaPeony_Buoy_N
  }
 ]
}
