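# Masked, two-sided lit material for the rope in the AquaPeony skydiving contrail, flagged for Niagara ribbons, sprites and mesh particles.
# A Tiling parameter repeats the mask and normal textures along U.
# The mask's red channel and AO Intensity darken the Base Color and Specular parameter, while its green channel supplies a cutout-style value.
# UseParticleColor blends in particle color and alpha multiplication for the color and green-channel values.
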Material /BRCosmetics/Athena/Effects/Particles/Skydiving/NS_Contrail_AquaPeony/Materials/M_Contrail_AquaPeony_Rope
graph {
 properties {
  bUsedWithNiagaraMeshParticles=true
  bUsedWithNiagaraRibbons=true
  bUsedWithNiagaraSprites=true
  blendMode=BLEND_Masked
  pixelDepthOffsetMode=PDOM_Legacy
  twoSided=true
  usedShadingModels=Default Lit
 }
 nodes [
  n1 = VectorParameter() {parameterName=Base Color and Specular, defaultValue=<rgba 0.3 0.3 0.3 0.5>}
  n2 = Multiply(A=TextureCoordinate(), B=AppendVector(A=ScalarParameter() {defaultValue=8, parameterName=Tiling}, B=Constant() {r=1}))
  n3 = TextureSample(UVs=@n2) {texture=/BRCosmetics/Athena/Effects/Particles/Skydiving/NS_Contrail_AquaPeony/Textures/T_Contrail_AquaPeony_Rope_M}
  n4 = OneMinus(None=@n3.R)
  n5 = ScalarParameter() {defaultValue=0.45, parameterName=AO Intensity}
  n6 = Multiply(A=AppendVector(A=@n1.RGB, B=@n1.RGB), B=OneMinus(None=Multiply(A=@n4, B=@n5)))
  ComponentMask(None=@n6) {a=true}
  n7 = ComponentMask(None=@n6) {b=true, g=true, r=true}
  n8 = ParticleColor()
  n9 = ScalarParameter() {parameterName=UseParticleColor}
  LinearInterpolate(A=@n7, B=Multiply(A=@n8.RGB, B=@n7), Alpha=@n9) {constAlpha=0}
  LinearInterpolate(A=@n3.G, B=Multiply(A=@n8.A, B=@n3.G), Alpha=@n9)
  TextureSample(UVs=@n2) {
   samplerType=SAMPLERTYPE_Normal
   texture=/BRCosmetics/Athena/Effects/Particles/Skydiving/NS_Contrail_AquaPeony/Textures/T_Contrail_AquaPeony_Rope_N
  }
 ]
}
