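# Masked, two-sided, lit material for Niagara mesh particles in the AquaPeony skydiving contrail, named for a rope hand and using pirate ship rigging textures.
# A UseParticleColor scalar blends particle color into both a dimmed diffuse branch, fed to MF_FX_EmissiveAdjustments, and a brightened, desaturated diffuse branch.
# Packed mask channels and a normal map supply further surface inputs, with the blue channel remapped by a lerp.
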Material /BRCosmetics/Athena/Effects/Particles/Skydiving/NS_Contrail_AquaPeony/Materials/M_Contrail_AquaPeony_RopeHand
graph {
 properties {
  bAutomaticallySetUsageInEditor=false
  bUsedWithNiagaraMeshParticles=true
  blendMode=BLEND_Masked
  pixelDepthOffsetMode=PDOM_Legacy
  twoSided=true
  usedShadingModels=Default Lit
 }
 nodes [
  n1 = TextureSample() {samplerType=SAMPLERTYPE_Masks, texture=/BRCosmetics/Athena/Effects/Materials/Skydiving/ARRGH/Textures/T_Pirate_Ship_Rigging_S_Contrail}
  LinearInterpolate(Alpha=@n1.B) {constA=0.1, constB=0.9}
  n2 = TextureSample() {texture=/BRCosmetics/Athena/Effects/Materials/Skydiving/ARRGH/Textures/T_Pirate_Ship_Rigging_D_Contrail}
  n3 = ParticleColor()
  n4 = Multiply(A=@n3.RGB, B=@n2.RGB)
  n5 = ScalarParameter() {parameterName=UseParticleColor}
  MaterialFunctionCall(Emissive=Multiply(A=LinearInterpolate(A=@n2.RGB, B=@n4, Alpha=@n5) {constAlpha=0}) {constB=0.2}) {materialFunction=&MF_FX_EmissiveAdjustments}
  Multiply(A=LinearInterpolate(B=@n3.RGB, Alpha=@n5) {constA=1, constAlpha=0}, B=Multiply(A=Desaturation(None=@n2.RGB)) {constB=1.5})
  n6 = TextureSample() {samplerType=SAMPLERTYPE_Normal, texture=/BRCosmetics/Athena/Effects/Materials/Skydiving/ARRGH/Textures/T_Pirate_Ship_Rigging_N_Contrail}
  Reroute(None=@n6.RGB)
  Reroute(None=@n1.G)
  Reroute(None=@n1.R)
  VectorParameter() {parameterName=Base Color, defaultValue=<rgba 1 1 1 1>}
 ]
}
