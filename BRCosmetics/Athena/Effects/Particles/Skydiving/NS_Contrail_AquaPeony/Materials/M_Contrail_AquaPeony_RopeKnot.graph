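# Masked, two-sided Default Lit material for rope knot mesh particles, used by the AquaPeony skydiving contrail Niagara effect.
# A "Base Color and Specular" vector parameter tints a grayscale knot texture.
# The product is split into RGB and alpha masks, apparently for base color and specular as the parameter name suggests.
# A separate normal map sample supplies the knot's surface detail.
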Material /BRCosmetics/Athena/Effects/Particles/Skydiving/NS_Contrail_AquaPeony/Materials/M_Contrail_AquaPeony_RopeKnot
graph {
 properties {
  bAutomaticallySetUsageInEditor=false
  bUsedWithNiagaraMeshParticles=true
  blendMode=BLEND_Masked
  pixelDepthOffsetMode=PDOM_Legacy
  twoSided=true
  usedShadingModels=Default Lit
 }
 nodes [
  n1 = VectorParameter() {parameterName=Base Color and Specular, defaultValue=<rgba 0.35 0.35 0.35 0>}
  n2 = TextureSample() {
   samplerType=SAMPLERTYPE_LinearGrayscale
   texture=/BRCosmetics/Athena/Effects/Particles/Skydiving/NS_Contrail_AquaPeony/Textures/T_Contrail_AquaPeony_RopeKnot_D
  }
  n3 = Multiply(A=AppendVector(A=@n1.RGB, B=@n1.RGB), B=@n2.R)
  ComponentMask(None=@n3) {b=true, g=true, r=true}
  ComponentMask(None=@n3) {a=true}
  TextureSample() {
   samplerType=SAMPLERTYPE_Normal
   texture=/BRCosmetics/Athena/Effects/Particles/Skydiving/NS_Contrail_AquaPeony/Textures/T_Contrail_AquaPeony_RopeKnot_N
  }
 ]
}
