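# Lit, masked, fully rough material for the rope ribbon in the AquaPeony skydiving contrail effect.
# A Tiling parameter repeats the rope mask and normal textures along U.
# The inverted mask red channel, scaled by AO Intensity, darkens the Base Color and Specular parameter as fake occlusion.
# That result is split into RGB and alpha, apparently for base color and specular.
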Material /BRCosmetics/Athena/Effects/Particles/Skydiving/NS_Contrail_AquaPeony/Materials/M_Contrail_AquaPeony_Rope_Ribbon
graph {
 properties {bAutomaticallySetUsageInEditor=false, bFullyRough=true, blendMode=BLEND_Masked, usedShadingModels=Default Lit}
 nodes [
  n1 = VectorParameter() {parameterName=Base Color and Specular, defaultValue=<rgba 0.5 0.5 0.5 0.5>}
  n2 = Multiply(A=TextureCoordinate(), B=AppendVector(A=ScalarParameter() {defaultValue=8, parameterName=Tiling}, B=Constant() {r=1}))
  n3 = OneMinus(None=TextureSample(UVs=@n2) {texture=/BRCosmetics/Athena/Effects/Particles/Skydiving/NS_Contrail_AquaPeony/Textures/T_Contrail_AquaPeony_Rope_M}.R)
  n4 = ScalarParameter() {defaultValue=0.45, parameterName=AO Intensity}
  n5 = Multiply(A=AppendVector(A=@n1.RGB, B=@n1.RGB), B=OneMinus(None=Multiply(A=@n3, B=@n4)))
  ComponentMask(None=@n5) {b=true, g=true, r=true}
  ComponentMask(None=@n5) {a=true}
  TextureSample(UVs=@n2) {
   samplerType=SAMPLERTYPE_Normal
   texture=/BRCosmetics/Athena/Effects/Particles/Skydiving/NS_Contrail_AquaPeony/Textures/T_Contrail_AquaPeony_Rope_N
  }
 ]
}
