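# Material function for the AquaPeony skydiving contrail that bends mesh particle vertices through World Position Offset, with an animatable aperture.
# Smoothstep masks on local X and Y position are scaled by a Dynamic Parameter Aperture to form two rotation angles.
# Two RotateAboutAxis nodes use axes from separate Dynamic Parameters around the mesh particle pivot.
# Their sum is added to the incoming material attributes' WPO; debug sine and remap nodes are left unconnected.
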MaterialFunction /BRCosmetics/Athena/Effects/Particles/Skydiving/NS_Contrail_AquaPeony/Materials/MF_Contrail_AquaPeony_WPO
graph {nodes [
  n1 = DynamicParameter() {parameterIndex=1, defaultValue=<rgba 0 0 0 0>, paramNames [XAxis X, XAxis Y, XAxis Z, Aperture]}
  n2 = LocalPosition()
  n3 = ComponentMask(None=SmoothStep(Value=@n2.XYZ) {constMax=-23, constMin=-12}) {g=true}
  n4 = OneMinus(None=@n2.XYZ)
  n5 = Saturate(None=ComponentMask(None=SmoothStep(Value=@n4) {constMax=-21.5, constMin=-10.5}) {g=true})
  n6 = Add(A=Multiply(A=@n1.Aperture, B=Saturate(None=@n3)), B=OneMinus(None=Multiply(A=@n1.Aperture, B=@n5)))
  n7 = SmoothStep(Value=@n2.XYZ) {constMax=-33, constMin=-13}
  n8 = Multiply(A=Multiply(A=@n1.Aperture), B=Saturate(None=ComponentMask(None=@n7) {r=true}))
  Add(A=@n6, B=@n8)
  Constant()
  n9 = GetMaterialAttributes(None=FunctionInput() {inputType=FunctionInput_MaterialAttributes}) {attributeGetTypes [#1]}
  n10 = NamedRerouteUsage()
  FunctionOutput(None=SetMaterialAttributes!(MaterialAttributes=@n9.MaterialAttributes, WorldPositionOffset=Add(A=@n9.WorldPositionOffset, B=@n10), EmissiveColor) {attributeSetTypes [#1, #2]})
  n11 = VectorParameter() {defaultValue=<rgba -1 1 -0.2 -0.01>}
  MaterialFunctionCall(
   Input=MaterialFunctionCall() {materialFunction=&DebugTimeSine}.Result
   Input Low=@n11.R
   Input High=@n11.R
   Target Low=@n11.R
   Target High=@n11.R
  ) {materialFunction=&RemapValueRange}
  n12 = MaterialFunctionCall(R=@n1.XAxis X, G=@n1.XAxis X, B=@n1.XAxis X) {materialFunction=&AppendMany}
  n13 = MaterialFunctionCall() {materialFunction=&ObjectPivotPoint}
  n14 = WorldPosition()
  n15 = DynamicParameter() {parameterIndex=2, defaultValue=<rgba 0 0 0 0>, paramNames [XAxis X, XAxis Y, XAxis Z, Aperture]}
  n16 = MaterialFunctionCall(R=@n15.XAxis X, G=@n15.XAxis X, B=@n15.XAxis X) {materialFunction=&AppendMany}
  NamedRerouteDeclaration(None=Add(
    A=RotateAboutAxis(NormalizedRotationAxis=@n12.RGB, RotationAngle=@n6, PivotPoint=@n13.Mesh Particle Pivot Location, Position=@n14.XYZ)
    B=RotateAboutAxis(NormalizedRotationAxis=@n16.RGB, RotationAngle=@n8, PivotPoint=@n13.Mesh Particle Pivot Location, Position=@n14.XYZ)
   )) {name=WPO, nodeColor=<rgba 0.1484 0 1 1>}
  ParticleDirection()
  ScalarParameter() {parameterName=Aperture, sliderMin=-0.03}
 ]}
