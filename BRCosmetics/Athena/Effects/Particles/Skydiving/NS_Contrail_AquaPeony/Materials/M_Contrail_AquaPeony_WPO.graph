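# Character material for the AquaPeony skydiving contrail effect that, going by its name, adds world position offset, usable on Niagara mesh particles and skeletal meshes.
# It works in material attributes, wrapping the shared 2023 character parent PreFX and PostFX functions around the contrail-specific WPO function.
# PreFX feeds its result through that function into PostFX, and its M output also drives PostFX's mask, position, distance blend and Fresnel inputs.
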Material /BRCosmetics/Athena/Effects/Particles/Skydiving/NS_Contrail_AquaPeony/Materials/M_Contrail_AquaPeony_WPO
graph {
 properties {
  bUseMaterialAttributes=true
  bUsedWithMeshDeformer=true
  bUsedWithNiagaraMeshParticles=true
  bUsedWithSkeletalMesh=true
  pixelDepthOffsetMode=PDOM_Legacy
  shadingModel=MSM_FromMaterialExpression
  usedShadingModels=Default Lit | Subsurface | Clear Coat | Subsurface Profile | Cloth
 }
 nodes [
  n1 = MaterialFunctionCall() {materialFunction=&MF_CharacterParent_2023_PreFX}
  MaterialFunctionCall(
   Material=MaterialFunctionCall!(In=@n1.Result) {materialFunction=&/MF_Contrail_AquaPeony_WPO}.Result
   M=@n1.M
   Pre Skinned Local Position=@n1.M
   Distance Blend=@n1.M
   Fresnel=@n1.M
  ) {materialFunction=&MF_CharacterParent_2023_PostFX}
 ]
}
