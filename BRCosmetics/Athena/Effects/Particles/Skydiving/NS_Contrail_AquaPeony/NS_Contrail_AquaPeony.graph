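# Niagara skydiving contrail for the AquaPeony cosmetic, with rope chains of buoys, a shark mesh and water spray attached to the player's skeletal mesh.
# RopeRibbon builds swaying ropes from bones, while Buoy and RopeKnot meshes read its particles to stick and orient along them.
# The shark follows a bone with velocity-driven orientation and banking.
# Dive value drives offsets, spray rates and motion, with separate front-end preview tuning.
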NiagaraSystem /BRCosmetics/Athena/Effects/Particles/Skydiving/NS_Contrail_AquaPeony/NS_Contrail_AquaPeony
dependencies {
 usedDataInterfaces [
  &NiagaraDataInterfaceArrayFloat3
  &NiagaraDataInterfaceArrayInt32
  &NiagaraDataInterfaceCurve
  &NiagaraDataInterfaceColorCurve
  &NiagaraDataInterfaceVectorCurve
  &NiagaraDataInterfaceArrayFloat
  &NiagaraDataInterfaceCamera
  &NiagaraDataInterfaceParticleRead
  &NiagaraDataInterfaceSkeletalMesh
 ]
 usedDynamicInputs [
  &MaskFloatBySpawnGroup
  &MaskVectorBySpawnGroup
  &SetBoolByFloatComparison
  &Divide_Float
 ]
 usedModules [
  &self:SystemSpawnScript.SystemScriptSource.SystemScriptGraph.NiagaraNodeAssignment_6.SetVariables_#1
  &self:SystemSpawnScript.SystemScriptSource.SystemScriptGraph.NiagaraNodeAssignment_8.SetVariables_#2
  &SystemState
  /Game/Effects/Niagara/Modules/Skydiving/FN_DivingAttributes/v6/FN_DivingAttributes
  /Game/Effects/Niagara/Modules/Contrails/FN_Contrail_Setup
  &self:ScratchModule
  &self:SystemSpawnScript.SystemScriptSource.SystemScriptGraph.NiagaraNodeAssignment_4.SetVariables_#3
  &self:SystemSpawnScript.SystemScriptSource.SystemScriptGraph.NiagaraNodeAssignment_1.SetVariables_#4
  &FN_Integrator
  &Timeline
  &self:SystemSpawnScript.SystemScriptSource.SystemScriptGraph.NiagaraNodeAssignment_3.SetVariables_#5
  &self:SystemSpawnScript.SystemScriptSource.SystemScriptGraph.NiagaraNodeAssignment_7.SetVariables_#6
  &self:SystemSpawnScript.SystemScriptSource.SystemScriptGraph.NiagaraNodeAssignment_5.SetVariables_#7
 ]
}
emitters {
 Buoy {
  emitterUpdateScript [
   &EmitterState(
    Inactive Response=::NewEnumerator1
    Life Cycle Mode=::NewEnumerator1
    Loop Behavior=::NewEnumerator1
    Loop Duration Mode=::NewEnumerator1
    UseLoopDelay=false
   )
   &SpawnBurst_Instantaneous(
    Spawn Count=&Multiply_Int(A=$System.ChainNumber, B=$System.LinkPerChain)
    Use Spawn Probability=false
    UseLoopCountLimit=false
   )
  ]
  particleSpawnScript [
   &FN_SetSpawnGroup(Number of Spawn Groups=2)
   &InitializeParticle(
    Color Mode=::NewEnumerator0
    Lifetime=5
    Mesh Scale Mode=::NewEnumerator4
    Sprite Size Mode=::NewEnumerator3
    Uniform Sprite Size=100
    UsePositionOffset=false
    Write Material Random=true
   )
   &PartitionParticles(
    Number Of Partitions=$System.ChainNumber
    Particles Per Partition:NiagaraInt32=$System.LinkPerChain
    Partition Mode:ENiagara_PartitionMode=::NewEnumerator1
   )
   /SetVariables(Particles.NormalizedChainPosition:NiagaraFloat=$Particles.PartitionParticles.Normalized Index In Partition) {isSetParameters=true, module=SetVariables_#8}
   &SkeletalMeshLocation(
    Bone / Socket Index=$Particles.PartitionParticles.Partition
    Bone Sampling Mode=::NewEnumerator4
    MaskSpawnGroup=false
    Sampled Position Offset=&Lerp_Vector(A=<0 0 0>, B=<0 0 100>, Alpha=$Particles.NormalizedChainPosition)
    Sampled Position Offset Coordinate Space:ENiagaraSkelMeshTransforms=::NewEnumerator1
    Skeletal Mesh=$System.Skeletal Mesh
   )
   /SetVariables(Particles.MeshOrientation=&MultiplyQuaternion(
     Quaternion A=&MultiplyQuaternion(
      Quaternion A=&OrientMeshToVector(
       Look At Direction=&NormalizeVector(
        Vector To Normalize=&Multiply_VectorByFloat(
         Vector=<0 0 1>
         Float=&Lerp_Float(A=-1, B=1, Alpha=&MakeFloatFromInt(Int=$Particles.PartitionParticles.Partition))
        )
        Safe Normalize=true
        Safe Fallback Threshold=0
        Safe Fallback Vector=<0 0 0.001>
       )
       Reference Axis Vector=<1 0 0>
      )
      Quaternion B=&MultiplyQuaternion(
       Quaternion A=&MakeQuatFromAxisAngle(
        Axis=<0 1 0>
        AngleInDegrees=&Lerp_Float(A=90, B=-90, Alpha=&MakeFloatFromInt(Int=$Particles.PartitionParticles.Partition))
       )
       Quaternion B=&MakeQuatFromAxisAngle(
        Axis=<0 0 1>
        AngleInDegrees=&Lerp_Float(
         A=180
         B=&Lerp_Float(A=270, B=90, Alpha=&MakeFloatFromInt(Int=$Particles.PartitionParticles.Partition))
         Alpha=&FloatFromCurve(
          FloatCurve=DataInterface() {propertyValues {Curve {keys [
              {arriveTangent=2.7188, interpMode=RCIM_Cubic, leaveTangent=2.7188, tangentMode=RCTM_User, time=-1, value=-1}
              {interpMode=RCIM_Cubic, tangentMode=RCTM_User, time=0, value=0}
              {arriveTangent=2.2136, interpMode=RCIM_Cubic, leaveTangent=2.2136, tangentMode=RCTM_User, time=1, value=0.8}
             ]}}}
          CurveIndex=&DotProduct(
           Vector A=&NormalizeVector(
            Vector To Normalize=&Multiply_Vector(
             A=&Subtract_Position(
              A=$Particles.Position
              B=&ReturnCameraPosition(Coordinate Space:ENiagaraCoordinateSpace=::Simulation, Camera Query:NiagaraDataInterfaceCamera=DataInterface())
             )
             B=<1 1 1>
            )
            Safe Normalize=true
            Safe Fallback Threshold=0
            Safe Fallback Vector=<0 0 0.001>
           )
           Vector B=<0 0 1>
          )
         )
        )
        NormalizeAxis=true
       )
      )
     )
     Quaternion B=<0 0 0 1>
    )) {isSetParameters=true, module=SetVariables_#9}
  ]
  particleUpdateScript [
   &ParticleState(Kill Particles When Lifetime Has Elapsed=false)
   &SkeletalMeshLocation(
    Bone / Socket Index=$Particles.PartitionParticles.Partition
    Bone Sampling Mode=::NewEnumerator4
    MaskSpawnGroup=false
    Position Sampling=::NewEnumerator2
    Sampled Position Offset=&Lerp_Vector(A=<20 -3 0>, B=<-20 3 1>, Alpha=&MakeFloatFromInt(Int=$Particles.PartitionParticles.Partition))
    Skeletal Mesh=$System.Skeletal Mesh
   ) {module=SkeletalMeshLocation001}
   &self:StickMeshToRibbon_03(
    Index:NiagaraInt32=$Particles.PartitionParticles.Index In Partition
    LinkOffset=3
    Particle Attribute Reader=DataInterface() {propertyValues {EmitterBinding {bindingMode=Other, emitterName=RopeRibbon}}}
    ParticlePerRibbon:NiagaraInt32=$System.ParticlePerRibbon
    Partition:NiagaraInt32=$Particles.PartitionParticles.Partition
    RibbonSpace=3
   ) {module=StickMeshToRibbon_03}
   &InterpolateOverTime(
    Rate Of Change:NiagaraFloat=&FN_FrontEndFloatModifier(Gameplay=1, Front End=10)
    Target Vector Value:Vector3f=&Multiply_VectorByFloat(
     Vector=&NormalizeVector(
      Vector To Normalize=&Multiply_Vector(
       A=&ReturnCameraVector(Camera Vectors=::NewEnumerator4, Coordinate Space=::Simulation, Origin Position=$Particles.Position)
       B=<1 1 1>
      )
      Safe Normalize=true
      Safe Fallback Threshold=0
      Safe Fallback Vector=<0 0 0.001>
     )
     Float=-1
    )
    Variable Type=::NewEnumerator2
   )
   &InterpolateOverTime(
    Rate Of Change=10
    Target Vector Value:Vector3f=&Multiply_VectorByFloat(Vector=$Particles.RibbonDirectionVector, Float=-1)
    Variable Type=::NewEnumerator2
   ) {module=InterpolateOverTime001}
   &self:OrientChainLinks_03(
    FrontVector:Vector3f=$Particles.InterpolateOverTime.InterpolateOverTime_V3_.Moving Average
    RightVector:Vector3f=<1 0 0>
    UpVector:Vector3f=$Particles.InterpolateOverTime001.InterpolateOverTime_V3_.Moving Average
   ) {module=OrientChainLinks_03}
   &SolveForcesAndVelocity(Clamp Velocity=false, Limit Acceleration=false)
   &DynamicMaterialParameters(
    Index 0 Param 1=$User.ParaGlideLeanAlpha
    Module.Index0Param0Write=true
    Module.Index0Param1Write=true
    Module.Index0Param2Write=true
    Module.Index0Param3Write=true
   )
   /SetVariables(Particles.VisibilityTag=&SpawnGroup()) {isSetParameters=true}
  ]
  rendererClasses [&NiagaraMeshRendererProperties]
  renderers [
   {rendererClass=&NiagaraMeshRendererProperties, rendererIndex=0}
   {rendererClass=&NiagaraMeshRendererProperties, rendererIndex=1}
  ]
 }
 RopeKnot {
  emitterUpdateScript [
   &EmitterState(
    Inactive Response=::NewEnumerator1
    Life Cycle Mode=::NewEnumerator1
    Loop Behavior=::NewEnumerator1
    Loop Duration Mode=::NewEnumerator1
    UseLoopDelay=false
   )
   &SpawnBurst_Instantaneous(Spawn Count=2, Use Spawn Probability=false, UseLoopCountLimit=false)
  ]
  particleSpawnScript [
   &FN_SetSpawnGroup(Number of Spawn Groups=2)
   &InitializeParticle(
    Color Mode=::NewEnumerator0
    Lifetime=5
    Mesh Scale Mode=::NewEnumerator3
    Sprite Size Mode=::NewEnumerator3
    Uniform Sprite Size=100
    UsePositionOffset=false
    Write Material Random=true
   )
   &PartitionParticles(Particles Per Partition=1, Partition Mode:ENiagara_PartitionMode=::NewEnumerator1)
   /SetVariables(Particles.NormalizedChainPosition:NiagaraFloat=$Particles.PartitionParticles.Normalized Index In Partition) {isSetParameters=true, module=SetVariables_#8}
   &SkeletalMeshLocation(
    Bone / Socket Index=$Particles.PartitionParticles.Partition
    Bone Sampling Mode=::NewEnumerator4
    MaskSpawnGroup=false
    Sampled Position Offset=&Lerp_Vector(A=<0 0 0>, B=<0 0 100>, Alpha=$Particles.NormalizedChainPosition)
    Sampled Position Offset Coordinate Space:ENiagaraSkelMeshTransforms=::NewEnumerator1
    Skeletal Mesh=$System.Skeletal Mesh
   )
   /SetVariables(Particles.MeshOrientation=&MultiplyQuaternion(
     Quaternion A=&MultiplyQuaternion(
      Quaternion A=&OrientMeshToVector(
       Look At Direction=&NormalizeVector(
        Vector To Normalize=&Multiply_VectorByFloat(
         Vector=<0 0 1>
         Float=&Lerp_Float(A=-1, B=1, Alpha=&MakeFloatFromInt(Int=$Particles.PartitionParticles.Partition))
        )
        Safe Normalize=true
        Safe Fallback Threshold=0
        Safe Fallback Vector=<0 0 0.001>
       )
       Reference Axis Vector=<1 0 0>
      )
      Quaternion B=&MultiplyQuaternion(
       Quaternion A=&MakeQuatFromAxisAngle(
        Axis=<0 1 0>
        AngleInDegrees=&Lerp_Float(A=90, B=-90, Alpha=&MakeFloatFromInt(Int=$Particles.PartitionParticles.Partition))
       )
       Quaternion B=&MakeQuatFromAxisAngle(
        Axis=<0 0 1>
        AngleInDegrees=&Lerp_Float(
         A=180
         B=&Lerp_Float(A=270, B=90, Alpha=&MakeFloatFromInt(Int=$Particles.PartitionParticles.Partition))
         Alpha=&FloatFromCurve(
          FloatCurve=DataInterface() {propertyValues {Curve {keys [
              {arriveTangent=2.7188, interpMode=RCIM_Cubic, leaveTangent=2.7188, tangentMode=RCTM_User, time=-1, value=-1}
              {interpMode=RCIM_Cubic, tangentMode=RCTM_User, time=0, value=0}
              {arriveTangent=2.2136, interpMode=RCIM_Cubic, leaveTangent=2.2136, tangentMode=RCTM_User, time=1, value=0.8}
             ]}}}
          CurveIndex=&DotProduct(
           Vector A=&NormalizeVector(
            Vector To Normalize=&Multiply_Vector(
             A=&Subtract_Position(
              A=$Particles.Position
              B=&ReturnCameraPosition(Coordinate Space:ENiagaraCoordinateSpace=::Simulation, Camera Query:NiagaraDataInterfaceCamera=DataInterface())
             )
             B=<1 1 1>
            )
            Safe Normalize=true
            Safe Fallback Threshold=0
            Safe Fallback Vector=<0 0 0.001>
           )
           Vector B=<0 0 1>
          )
         )
        )
        NormalizeAxis=true
       )
      )
     )
     Quaternion B=<0 0 0 1>
    )) {isSetParameters=true, module=SetVariables_#9}
  ]
  particleUpdateScript [
   &ParticleState(Kill Particles When Lifetime Has Elapsed=false)
   &SkeletalMeshLocation(
    Bone / Socket Index=$Particles.PartitionParticles.Partition
    Bone Sampling Mode=::NewEnumerator4
    MaskSpawnGroup=false
    Position Sampling=::NewEnumerator2
    Sampled Position Offset=&Lerp_Vector(A=<20 -3 0>, B=<-20 3 1>, Alpha=&MakeFloatFromInt(Int=$Particles.PartitionParticles.Partition))
    Skeletal Mesh=$System.Skeletal Mesh
   ) {module=SkeletalMeshLocation001}
   &self:StickMeshToRibbon_03(
    Index:NiagaraInt32=$Particles.PartitionParticles.Index In Partition
    LinkOffset=20
    Particle Attribute Reader=DataInterface() {propertyValues {EmitterBinding {bindingMode=Other, emitterName=RopeRibbon}}}
    ParticlePerRibbon:NiagaraInt32=$System.ParticlePerRibbon
    Partition:NiagaraInt32=$Particles.PartitionParticles.Partition
    RibbonSpace=3
   ) {module=StickMeshToRibbon_03}
   &InterpolateOverTime(
    Rate Of Change:NiagaraFloat=&FN_FrontEndFloatModifier(Gameplay=1, Front End=10)
    Target Vector Value:Vector3f=&Multiply_VectorByFloat(
     Vector=&NormalizeVector(
      Vector To Normalize=&Multiply_Vector(
       A=&ReturnCameraVector(Camera Vectors=::NewEnumerator4, Coordinate Space=::Simulation, Origin Position=$Particles.Position)
       B=<1 1 1>
      )
      Safe Normalize=true
      Safe Fallback Threshold=0
      Safe Fallback Vector=<0 0 0.001>
     )
     Float=-1
    )
    Variable Type=::NewEnumerator2
   )
   &InterpolateOverTime(
    Rate Of Change=10
    Target Vector Value:Vector3f=&Multiply_VectorByFloat(Vector=$Particles.RibbonDirectionVector, Float=-1)
    Variable Type=::NewEnumerator2
   ) {module=InterpolateOverTime001}
   &self:OrientChainLinks_03(
    FrontVector:Vector3f=$Particles.InterpolateOverTime.InterpolateOverTime_V3_.Moving Average
    RightVector:Vector3f=<1 0 0>
    UpVector:Vector3f=$Particles.InterpolateOverTime001.InterpolateOverTime_V3_.Moving Average
   ) {module=OrientChainLinks_03}
   &SolveForcesAndVelocity(Clamp Velocity=false, Limit Acceleration=false)
   &DynamicMaterialParameters(
    Index 0 Param 1=$User.ParaGlideLeanAlpha
    Module.Index0Param0Write=true
    Module.Index0Param1Write=true
    Module.Index0Param2Write=true
    Module.Index0Param3Write=true
   )
  ]
  rendererClasses [&NiagaraMeshRendererProperties]
  renderers [{rendererClass=&NiagaraMeshRendererProperties, rendererIndex=0}]
 }
 RopeRibbon {
  emitterUpdateScript [
   &EmitterState(
    Inactive Response=::NewEnumerator1
    Life Cycle Mode=::NewEnumerator1
    Loop Behavior=::NewEnumerator1
    Loop Duration Mode=::NewEnumerator1
    UseLoopDelay=false
   )
   &SpawnBurst_Instantaneous(
    Spawn Count=&Multiply_Int(A=$System.ChainNumber, B=$System.ParticlePerRibbon)
    Use Spawn Probability=false
    UseLoopCountLimit=false
   )
  ]
  particleSpawnScript [
   &InitializeParticle(
    Lifetime=5
    Mesh Scale Mode=::NewEnumerator4
    Ribbon Width=2.25
    Ribbon Width Mode=::NewEnumerator1
    Sprite Size Mode=::NewEnumerator3
    UsePositionOffset=false
    Write Material Random=true
   )
   &PartitionParticles(
    Number Of Partitions=$System.ChainNumber
    Particles Per Partition:NiagaraInt32=$System.ParticlePerRibbon
    Partition Mode:ENiagara_PartitionMode=::NewEnumerator1
   )
   /SetVariables(Particles.NormalizedChainPosition:NiagaraFloat=$Particles.PartitionParticles.Normalized Index In Partition) {isSetParameters=true, module=SetVariables_#8}
   /SetVariables() {isSetParameters=true}
  ]
  particleUpdateScript [
   &ParticleState(Kill Particles When Lifetime Has Elapsed=false)
   &SkeletalMeshLocation(
    Bone / Socket Index=$Particles.PartitionParticles.Partition
    Bone Sampling Mode=::NewEnumerator4
    MaskSpawnGroup=false
    Orientation Sampling=::NewEnumerator1
    Sampled Position Offset=&FN_FrontEndVector3Modifier(Gameplay=<-1 -2 0>, Front End=&Lerp_Vector(A=<-1 -2 0>, B=<0 4 0>, Alpha=$System.FN_Contrail_Setup.Dive Value))
    Sampled Position Offset Coordinate Space:ENiagaraSkelMeshTransforms=::NewEnumerator2
    Skeletal Mesh=$System.Skeletal Mesh
   ) {module=SkeletalMeshLocation002}
   /SetVariables(Particles.Position=&AddVectorToPosition(
     Position=$Particles.Position
     Vector=&Multiply_VectorByFloat(
      Vector=&Add_Vector(
       A=&Multiply_Vector(
        A=&NormalizeVector(
         Vector To Normalize=&FN_FrontEndVector3Modifier(Gameplay=$System.PlayerVelDir, Front End=$System.FrontendVelocity)
         Safe Normalize=true
         Safe Fallback Threshold=0
         Safe Fallback Vector=<0 0 0.001>
        )
        B=&SelectVectorFromArray(
         Direct Array Index=$Particles.PartitionParticles.Partition
         Vector Selection Array=DataInterface() {propertyValues {FloatData [<-0.5 -0.5 -1>, <-0.5 -0.5 -1>, <0.1 -1.5 -1>, <0.1 -0.5 -1>, <-0.5 -1.5 -1>, <-0.5 -1 -1>, <0.1 -1 -1>, <0.1 -0.5 -1>]}}
        )
       )
       B=&Multiply_VectorByFloat(
        Vector=&TransformVector(
         Vector=&SelectVectorFromArray(
          Direct Array Index=$Particles.PartitionParticles.Partition
          Direct Array Mode=::NewEnumerator1
          Vector Selection Array=DataInterface() {propertyValues {FloatData [<2 -2 0>, <-2 0 0>, <1 2 0>, <-0.5 0 0>, <0.5 0 0>, <-1 2 0>, <1 1 0>, <-2 1 0>]}}
         )
         Destination Space=::Simulation
        )
        Float=&SelectFloatFromArray(
         Direct Array Index=&SpawnGroup()
         Float Selection Array=DataInterface() {propertyValues {FloatData [0.1,0.15,0.08,0.16,0.15,0.18]}}
        )
       )
      )
      Float=&FloatFromCurve(
       FloatCurve=DataInterface() {propertyValues {Curve {keys [{tangentMode=RCTM_User, time=0, value=0}, {tangentMode=RCTM_User, time=1, value=1}]}}}
       CurveIndex=$Particles.PartitionParticles.Normalized Index In Partition
       Scale Curve=&SelectFloatFromArray(
        Direct Array Index=&SpawnGroup()
        Float Selection Array=DataInterface() {propertyValues {FloatData [97,97,120,100,75,55,94,80]}}
       )
      )
     )
    )) {isSetParameters=true}
   /SetVariables(Particles.Position=&AddVectorToPosition(
     Position=$Particles.Position
     Vector=&Multiply_VectorByFloat(
      Vector=/Niagara/DynamicInputs/Vectors/RotateVector(
       Vector To Rotate:Vector3f=<10 0 3>
       Yaw:NiagaraFloat=&Add_Float(
        A=$Particles.Age
        B=&Multiply_Float(A=&MakeFloatFromInt(Int=&SpawnGroup()), B=2)
       )
       Pitch:NiagaraFloat=0
       Roll:NiagaraFloat=0
       Rotation Rate:NiagaraFloat=&SelectFloatFromArray(
        Direct Array Index=$Particles.PartitionParticles.Partition
        Float Selection Array=DataInterface() {propertyValues {FloatData [100,100,130,120,160,150,120,140]}}
       )
      )
      Float=&Multiply_Float(
       A=&Sine(
        Normalized Angle=&Add_Float(
         A=&Multiply_Float(
          A=$Emitter.Age
          B=&SelectFloatFromArray(
           Direct Array Index=&SpawnGroup()
           Float Selection Array=DataInterface() {propertyValues {FloatData [-2,-2.5,-2.5,-3,-3,-2,-2.5,-2]}}
          )
         )
         B=$Particles.PartitionParticles.Normalized Index In Partition
        )
        Period=0.9
        Scale=&Lerp_Float(A=0.8, B=1.4, Alpha=$System.FN_Contrail_Setup.Dive Value)
        Bias=0
       )
       B=&FloatFromCurve(
        FloatCurve=DataInterface() {propertyValues {Curve {keys [
            {arriveTangent=1.1676, interpMode=RCIM_Cubic, leaveTangent=1.1676, tangentMode=RCTM_User, time=0, value=0}
            {arriveTangent=1.4544, interpMode=RCIM_Cubic, leaveTangent=1.4544, tangentMode=RCTM_User, time=1, value=1}
           ]}}}
        CurveIndex=$Particles.PartitionParticles.Normalized Index In Partition
       )
      )
     )
    )) {isSetParameters=true}
   /SetVariables(Particles.Position=&AddVectorToPosition(
     Position=$Particles.Position
     Vector=&Multiply_VectorByFloat(
      Vector=&SelectVectorFromArray(
       Direct Array Index=$Particles.PartitionParticles.Partition
       Direct Array Mode=::NewEnumerator1
       Vector Selection Array=DataInterface() {propertyValues {FloatData [<-20 40 130>, <40 20 130>, <-25 0 0>, <25 0 0>]}}
      )
      Float=&FloatFromCurve(
       FloatCurve=DataInterface() {propertyValues {Curve {keys [{arriveTangent=0.9713, time=0, value=0}, <at 0.125 0.2>, <at 0.95 0.92>, {interpMode=RCIM_Cubic, leaveTangent=0.018, time=1, value=1}]}}}
       CurveIndex=$Particles.Initial.PartitionParticles.Normalized Index In Partition
      )
     )
    )) {isSetParameters=true}
  ]
  rendererClasses [&NiagaraRibbonRendererProperties]
  renderers [{rendererClass=&NiagaraRibbonRendererProperties}]
 }
 Shark {
  emitterUpdateScript [
   &EmitterState()
   &SpawnBurst_Instantaneous(
    Age:NiagaraFloat=$Emitter.Age
    Spawn Count=&MakeIntFromFloat(Float=&FN_FrontEndFloatModifier(Gameplay=1, Front End=1))
    Use Spawn Probability=false
    UseLoopCountLimit=false
   )
   &InterpolateOverTime(Rate Of Change=50, Target Vector Value:Vector3f=$Engine.Owner.Velocity, Variable Type=::NewEnumerator2)
  ]
  particleSpawnScript [&InitializeParticle(
    Mesh Scale Mode=::NewEnumerator3
    Mesh Uniform Scale=&FN_FrontEndFloatModifier(Gameplay=3, Front End=2.5)
    UsePositionOffset=false
    Write Material Random=true
   )]
  particleUpdateScript [
   &ParticleState(Kill Particles When Lifetime Has Elapsed=false, Let Infinitely Lived Particles Die When Emitter Deactivates=true)
   /SetVariables(Particles.AmbientFrequency=1.5) {isSetParameters=true}
   /SetVariables(Particles.Ambient:Vector3f=&Multiply_VectorByFloat(
     Vector=&MakeVector(
      X=&Waveform(
       "Waveform [1]"=::NewEnumerator12
       "[1] Input Position"=&RandomRangeFloat(Minimum=0, Maximum=1000)
       "[1] Frequency"=$Particles.AmbientFrequency
       "[1] Amplitude Min/Max"=<-50 50>
      )
      Y=&Waveform(
       "Waveform [1]"=::NewEnumerator12
       "[1] Input Position"=&RandomRangeFloat(Minimum=1200, Maximum=1800)
       "[1] Frequency"=$Particles.AmbientFrequency
       "[1] Amplitude Min/Max"=<-50 50>
      )
      Z=&Waveform(
       "Waveform [1]"=::NewEnumerator12
       "[1] Input Position"=&RandomRangeFloat(Minimum=2200, Maximum=2800)
       "[1] Frequency"=$Particles.AmbientFrequency
       "[1] Amplitude Min/Max"=<-50 50>
      )
     )
     Float=&Lerp_Float(A=0.02, B=0.1, Alpha=$System.FN_Contrail_Setup.Dive Value)
    )) {isSetParameters=true}
   &InterpolateOverTime(
    Rate Of Change=50
    Target Float Value:NiagaraFloat=&MakeCustomFloatFromBool(Bool=$System.Timeline001.IsRewinding, True Float=0, False Float=1)
   ) {module=InterpolateOverTime004}
   &SkeletalMeshLocation(
    Bone / Socket Index=11
    Bone Sampling Mode=::NewEnumerator4
    MaskSpawnGroup=false
    Sampled Position Offset=&Add_Vector(
     A=$Particles.CatchupVector
     B=&RotateVector(
      Vector To Rotate:Vector3f=&AddVectorToPosition(
       Position=&MakeVector(
        X=0
        Y=&FloatFromCurve(
         FloatCurve=DataInterface() {propertyValues {Curve {keys [
             {arriveTangent=0.0815, interpMode=RCIM_Cubic, leaveTangent=0.0815, tangentMode=RCTM_User, time=0, value=0}
             {arriveTangent=-2.3986, interpMode=RCIM_Cubic, leaveTangent=-2.3986, tangentMode=RCTM_User, time=1, value=-2}
            ]}}}
         CurveIndex=$System.FN_Contrail_Setup.Dive Value
         Scale Curve=30
        )
        Z=&FloatFromCurve(
         FloatCurve=DataInterface() {propertyValues {Curve {keys [
             {arriveTangent=0.0815, interpMode=RCIM_Cubic, leaveTangent=0.0815, tangentMode=RCTM_User, time=0, value=-1}
             {arriveTangent=-2.3986, interpMode=RCIM_Cubic, leaveTangent=-2.3986, tangentMode=RCTM_User, time=1, value=-2}
            ]}}}
         CurveIndex=$System.FN_Contrail_Setup.Dive Value
         Scale Curve=&FN_FrontEndFloatModifier(Gameplay=50, Front End=0)
        )
       )
       Vector=&Multiply_VectorByFloat(Vector=$System.Vector, Float=2)
      )
      Yaw:NiagaraFloat=&Waveform(
       "Waveform [1]"=::NewEnumerator12
       "[1] Input Position"=&RandomRangeFloat(Minimum=200000, Maximum=100000)
       "[1] Frequency"=1.25
       "[1] Amplitude Min/Max"=<7 -7>
       "[1] Scale"=0.2
      )
      Pitch:NiagaraFloat=0
      Roll:NiagaraFloat=0
      Rotation Rate:NiagaraFloat=1
     )
    )
    Skeletal Mesh=$System.Skeletal Mesh
    Use Sampled Mesh Attribute as Ribbon ID=false
   )
   &ScaleMeshSize(Scale Factor=&VectorFromFloat(Value=&FloatFromCurve(
      FloatCurve=DataInterface() {propertyValues {Curve {keys [{arriveTangent=1.0186, interpMode=RCIM_Cubic, leaveTangent=1.0186, tangentMode=RCTM_User, time=0, value=0}, <at 0.3 1 RCIM_Cubic>]}}}
      CurveIndex=$Particles.Age
     )))
   &SolveForcesAndVelocity(Clamp Velocity=false, Limit Acceleration=false)
   &UpdateMeshOrientation(
    Bank Rate:NiagaraFloat=&Lerp_Float(A=1, B=900, Alpha=$System.FN_Contrail_Setup.Dive Value)
    Banking Angle Min / Max:Vector2f=&Lerp_Vector2(A=<0 0>, B=<0 0>, Alpha=$System.FN_Contrail_Setup.Dive Value)
    "Foward Vector (Mesh Space)":Vector3f=&Lerp_Vector(A=<0.3 0 1>, B=<1 0 0.5>, Alpha=$System.FN_Contrail_Setup.Dive Value)
    Initial Forward Vector:Vector3f=<0 -1 0>
    Orientation Method=::NewEnumerator3
    Particle Velocity:Vector3f=$Emitter.InterpolateOverTime.InterpolateOverTime_V3_.Moving Average
    "Right Vector (Mesh Space)":Vector3f=&Lerp_Vector(
     A=<0 1 -0.05>
     B=<0 1 0.05>
     Alpha=&MakeFloatFromVector(Vector=$System.Vector, Channel=::NewEnumerator0)
    )
    Rotation Decay Rate In Degrees:NiagaraFloat=&Lerp_Float(A=150, B=500, Alpha=$System.Dive)
    Up Vector:Vector3f=&Lerp_Vector(
     A=&ReturnMeshOrientationAxis(Mesh Orientation Axis=::YAxis, Mesh Orientation Quat=$Particles.SkeletalMeshLocation.SampledRotation)
     B=&ReturnMeshOrientationAxis(Mesh Orientation Axis=::ZAxis, Mesh Orientation Quat=$Particles.SkeletalMeshLocation.SampledRotation)
     Alpha=&FloatFromCurve(FloatCurve=DataInterface() {propertyValues {Curve {keys [<at 0 -1>,<at 0.25 0.5>]}}}, CurveIndex=$System.FN_Contrail_Setup.Dive Value)
    )
    Upright Roll Correction Speed:NiagaraFloat=&Lerp_Float(A=25, B=90000, Alpha=$System.Dive)
    Velocity Orientation Convergence Rate:NiagaraFloat=&Lerp_Float(
     A=&Lerp_Float(
      A=3
      B=12
      Alpha=&Waveform(
       "Waveform [1]"=::NewEnumerator12
       "[1] Input Position"=&RandomRangeFloat(Minimum=0, Maximum=100)
       "[1] Frequency"=5
       "[1] Amplitude Min/Max"=<0 1>
      )
     )
     B=25
     Alpha=$System.Dive
    )
   ) {module=UpdateMeshOrientation001}
   /SetVariables(Particles.FrontendRotation:Quat4f=&MultiplyQuaternion(
     Quaternion A=/CascadeToNiagaraConverter/NiagaraScripts/DynamicInputScripts/MakeQuatFromEuler(
      Euler:Vector3f=&Lerp_Vector(
       A=<1 1 90>
       B=<-1 -1 90>
       Alpha=&MakeFloatFromVector(Vector=$Particles.Ambient, Channel=::NewEnumerator0)
      )
      Scale:NiagaraFloat=1
     )
     Quaternion B=$Particles.SkeletalMeshLocation.SampledRotation
    )) {isSetParameters=true}
   /SetVariables(Particles.FrontendRotation:Quat4f=&MultiplyQuaternion(
     Quaternion A=$StackContext.FrontendRotation
     Quaternion B=&MakeQuaternion(
      Quaternion From=0
      Axis=<0 1 0>
      Angle=&Lerp_Float(A=0, B=30, Alpha=$System.FN_Contrail_Setup.Dive Value)
      From=?
      To=?
      Angle Type=::NewEnumerator0
      Coordinate Space=::World
      Normalize Vectors=false
     )
    )) {isSetParameters=true}
   /SetVariables(Particles.BankOrientation:Quat4f=/Niagara/DynamicInputs/TypeConversions/MakeCustomQuaternionFromBool(
     Bool:NiagaraBool=$User.bIsFrontEndPreview
     True Quat:Quat4f=$StackContext.FrontendRotation
     False Quat:Quat4f=&MultiplyQuaternion(
      Quaternion A=$Particles.MeshOrientation
      Quaternion B=&MakeQuaternion(
       Quaternion From=0
       Axis=&Lerp_Vector(A=<0.5 0 0>, B=<0.5 0 -0.7>, Alpha=$System.FN_Contrail_Setup.Dive Value)
       Angle=$System.BankAngle
       From=?
       To=?
       Angle Type=::NewEnumerator0
       Coordinate Space=::World
       Normalize Vectors=false
      )
     )
    )) {isSetParameters=true}
   &InterpolateOverTime(Rate Of Change=10, Target Quat Value:Quat4f=$Particles.BankOrientation, Variable Type=::NewEnumerator5) {module=InterpolateOverTime002}
   /SetVariables(Particles.New Orient:Quat4f=$Particles.InterpolateOverTime002.InterpolateOverTime_Quaternion.Moving Average) {isSetParameters=true}
   /SetVariables(Particles.MeshOrientation=$Particles.New Orient) {isSetParameters=true}
   &DynamicMaterialParameters(
    Index 0 Param 1=&SelectFloatFromArray(
     Direct Array Index=&ReturnExecIndex()
     Float Selection Array=DataInterface() {propertyValues {FloatData [0,0.5]}}
    )
    Index 0 Param 2=&RandomRangeFloat(Minimum=0, Maximum=1)
    Index 0 Param 3=0.4
    Index 1 Param 1=&MakeFloatFromVector(
     Vector=&ReturnMeshOrientationAxis(Mesh Orientation Axis=::XAxis, Mesh Orientation Quat=$Particles.MeshOrientation)
     Channel=::NewEnumerator0
    )
    Index 1 Param 2=&MakeFloatFromVector(
     Vector=&ReturnMeshOrientationAxis(Mesh Orientation Axis=::XAxis, Mesh Orientation Quat=$Particles.MeshOrientation)
     Channel=::NewEnumerator1
    )
    Index 1 Param 3=&MakeFloatFromVector(
     Vector=&ReturnMeshOrientationAxis(Mesh Orientation Axis=::XAxis, Mesh Orientation Quat=$Particles.MeshOrientation)
     Channel=::NewEnumerator2
    )
    Index 1 Param 4=&Waveform(
     Global Amplitude Scale=&Lerp_Float(A=1, B=4, Alpha=$System.FN_Contrail_Setup.Dive Value)
     "Waveform [1]"=::NewEnumerator2
     "[1] Frequency"=0.8
     "[1] Compound Frequency"=1.8
     "[1] Amplitude Min/Max"=<0.015 -0.015>
    )
    Index 2 Param 1=&MakeFloatFromVector(
     Vector=&ReturnMeshOrientationAxis(Mesh Orientation Axis=::ZAxis, Mesh Orientation Quat=$Particles.MeshOrientation)
     Channel=::NewEnumerator0
    )
    Index 2 Param 2=&MakeFloatFromVector(
     Vector=&ReturnMeshOrientationAxis(Mesh Orientation Axis=::ZAxis, Mesh Orientation Quat=$Particles.MeshOrientation)
     Channel=::NewEnumerator1
    )
    Index 2 Param 3=&MakeFloatFromVector(
     Vector=&ReturnMeshOrientationAxis(Mesh Orientation Axis=::ZAxis, Mesh Orientation Quat=$Particles.MeshOrientation)
     Channel=::NewEnumerator2
    )
    Module.Param0WriteEnabled=true
    Module.Param0WriteEnabled1=true
    Module.Param0WriteEnabled2=true
    Module.Param1WriteEnabled=true
    Module.Param1WriteEnabled1=true
    Module.Param1WriteEnabled2=true
    Module.Param2WriteEnabled=true
    Module.Param2WriteEnabled1=true
    Module.Param2WriteEnabled2=true
    Module.Param3WriteEnabled=true
    Module.Param3WriteEnabled1=true
    Module.Param3WriteEnabled2=true
    Write Parameter Index 1=true
    Write Parameter Index 2=true
   )
   &ScaleColor(
    Scale Alpha=&FN_FrontEndFloatModifier(Gameplay=2, Front End=2)
    Scale RGB=&MakeVectorFromLinearColorRGB(Color=&FN_Select_Color_from_Time_Of_Day(Morning=<rgba 0 0.1417 0.85 1>, Day=<rgba 0 0.25 1 1>, Evening=<rgba 0 0.1167 0.7 1>, Night=<rgba 0 0.075 0.9 1>))
    ScaleA=true
    ScaleRGB=true
   )
   /SetVariables(
    Particles.DirectionalColor:LinearColor=&MakeLinearColorFromVectorAndFloat(
     "Vector (RGB)"=&MakeVectorFromLinearColorRGB(Color=$Particles.Color)
     "Float (Alpha)"=&FloatFromCurve(
      FloatCurve=DataInterface() {propertyValues {Curve {keys [<at 0 0>,<at 1 1>]}}}
      CurveIndex=$System.FN_Contrail_Setup.Dive Value
      Scale Curve=3.5
     )
    )
    Particles.DirectionalDyn:Vector4f=&MakeVector4(
     X=0
     Y=0
     Z=&FloatFromCurve(FloatCurve=DataInterface() {propertyValues {Curve {keys [<at 0 5>,<at 1 0.4>]}}}, CurveIndex=$System.FN_Contrail_Setup.Dive Value)
     W=0
    )
   ) {isSetParameters=true}
   /SetVariables(Particles.BackWakeColor:LinearColor=&MakeLinearColorFromVectorAndFloat(
     "Vector (RGB)"=&MakeVectorFromLinearColorRGB(Color=$Particles.Color)
     "Float (Alpha)"=&FloatFromCurve(
      FloatCurve=DataInterface() {propertyValues {Curve {keys [<at 0 0>,<at 1 1>]}}}
      CurveIndex=$System.FN_Contrail_Setup.Dive Value
      Scale Curve=&FN_FrontEndFloatModifier(Gameplay=3.5, Front End=2)
     )
    )) {isSetParameters=true}
  ]
  rendererClasses [&NiagaraMeshRendererProperties]
  renderers [
   {rendererClass=&NiagaraMeshRendererProperties, rendererIndex=0}
   {rendererClass=&NiagaraMeshRendererProperties, rendererIndex=1}
   {rendererClass=&NiagaraMeshRendererProperties, rendererIndex=2}
   {rendererClass=&NiagaraMeshRendererProperties, rendererIndex=3}
   {rendererClass=&NiagaraMeshRendererProperties, rendererIndex=4}
   {rendererClass=&NiagaraMeshRendererProperties, rendererIndex=5}
  ]
 }
 WaterParticles {
  emitterUpdateScript [
   &EmitterState(
    Life Cycle Mode=::NewEnumerator1
    Loop Behavior=::NewEnumerator1
    Loop Delay=$System.Delay
    Loop Duration Mode=::NewEnumerator1
    UseLoopDelay=false
   )
   &SpawnRate(SpawnRate=&Lerp_Float(A=25, B=0, Alpha=$System.FN_Contrail_Setup.Dive Value), Use Spawn Probability=false)
   &SpawnRate(
    SpawnGroup=1
    SpawnRate=&Lerp_Float(A=50, B=150, Alpha=$System.FN_Contrail_Setup.Dive Value)
    Use Spawn Probability=false
   ) {module=SpawnRate001}
   &SpawnRate(
    SpawnGroup=2
    SpawnRate=&Lerp_Float(A=30, B=200, Alpha=$System.FN_Contrail_Setup.Dive Value)
    Use Spawn Probability=false
   ) {module=SpawnRate003}
  ]
  particleSpawnScript [
   /SetVariables(Particles.Offset:Vector3f=&AddVectorToPosition(
     Position=&AddVectorToPosition(
      Position=&MakeVector(
       X=0
       Y=&FloatFromCurve(
        FloatCurve=DataInterface() {propertyValues {Curve {keys [
            {arriveTangent=0.0815, interpMode=RCIM_Cubic, leaveTangent=0.0815, tangentMode=RCTM_User, time=0, value=0}
            {arriveTangent=-2.3986, interpMode=RCIM_Cubic, leaveTangent=-2.3986, tangentMode=RCTM_User, time=1, value=-2}
           ]}}}
        CurveIndex=$System.FN_Contrail_Setup.Dive Value
        Scale Curve=30
       )
       Z=&FloatFromCurve(
        FloatCurve=DataInterface() {propertyValues {Curve {keys [
            {arriveTangent=0.0815, interpMode=RCIM_Cubic, leaveTangent=0.0815, tangentMode=RCTM_User, time=0, value=-1}
            {arriveTangent=-2.3986, interpMode=RCIM_Cubic, leaveTangent=-2.3986, tangentMode=RCTM_User, time=1, value=-2}
           ]}}}
        CurveIndex=$System.FN_Contrail_Setup.Dive Value
        Scale Curve=&FN_FrontEndFloatModifier(Gameplay=50, Front End=0)
       )
      )
      Vector=&Multiply_VectorByFloat(Vector=$System.Vector, Float=2)
     )
     Vector=&Lerp_Vector(A=<0 0 20>, B=<0 10 10>, Alpha=$System.FN_Contrail_Setup.Dive Value)
    )) {isSetParameters=true}
   &InitializeParticle(
    Color=&MergeHelper_ColorSwitch(
     0=&FN_Select_Color_from_Time_Of_Day(Morning=<rgba 0 0.1417 0.85 1>, Day=<rgba 0 0.25 1 1>, Evening=<rgba 0 0.1167 0.7 1>, Night=<rgba 0 0.075 0.9 1>)
     1=<rgba 1 1 1 1>
     2=<rgba 5 5 5 1>
     3=<rgba 0 0 0 0>
     4=<rgba 0 0 0 0>
     Index=&SpawnGroup()
     Num=::NewEnumerator3
    )
    Lifetime=&MergeHelper_FloatSwitch(
     0=&Lerp_Float(
      A=&RandomRangeFloat(Minimum=0.25, Maximum=0.45)
      B=&RandomRangeFloat(Minimum=0.2, Maximum=0.4)
      Alpha=$System.FN_Contrail_Setup.Dive Value
     )
     1=&RandomRangeFloat(Minimum=0.15, Maximum=0.25)
     2=&RandomRangeFloat(Minimum=0.5, Maximum=1)
     3=0
     4=0
     Index=0
     Num=::NewEnumerator3
    )
    Mesh Scale Max=<0.35 0.4 0.2>
    Mesh Scale Min=<0.25 0.5 0.2>
    Mesh Scale Mode=::NewEnumerator2
    Position Mode=::NewEnumerator0
    Sprite Size=&MergeHelper_Vector2Switch(
     0=<0 0>
     1=&RandomRangeVector2D(Minimum=<25 25>, Maximum=<35 35>)
     2=&RandomRangeVector2D(Minimum=<0.2 3>, Maximum=<2 10>)
     3=<0 0>
     4=<0 0>
     Index=&SpawnGroup()
     Num=::NewEnumerator3
    )
    Sprite Size Mode=::NewEnumerator4
    Write Material Random=true
   )
   &DynamicMaterialParameters(
    Index 0 Param 1=&RandomRangeFloat(Minimum=0, Maximum=1)
    Module.Index0Param0Write=true
    Module.Index0Param1Write=true
    Module.Index0Param2Write=true
    Module.Index0Param3Write=true
   )
   &ShapeLocation(
    MaskSpawnGroup=false
    Non Uniform Scale=<1 1 0>
    Offset=$Particles.Offset
    Offset Mode=::NewEnumerator0
    Rotation Angle Type:ENiagara_AngleInput=::NewEnumerator0
    Rotation Coordinate Space=::Simulation
    Rotation Mode=::NewEnumerator2
    Sphere Distribution=::NewEnumerator1
    Sphere Radius=&Multiply_Float(
     A=&MergeHelper_FloatSwitch(
      0=&FN_FrontEndFloatModifier(Gameplay=47, Front End=42)
      1=55
      2=52
      3=?
      Index=&SpawnGroup()
      Num=::NewEnumerator1
     )
     B=&FN_FrontEndFloatModifier(Gameplay=1, Front End=0.775)
    )
    U Position=&RandomRangeFloat(Minimum=0, Maximum=1)
    UseOwnerScale=true
    Yaw / Pitch / Roll=&MakeVector(
     X=&Lerp_Float(A=0, B=&Lerp_Float(A=0.3, B=-0.3, Alpha=$System.BankAngle), Alpha=$System.FN_Contrail_Setup.Dive Value)
     Y=0
     Z=&Lerp_Float(
      A=&FN_FrontEndFloatModifier(Gameplay=50, Front End=20)
      B=&FN_FrontEndFloatModifier(Gameplay=20, Front End=6)
      Alpha=&FloatFromCurve(
       FloatCurve=DataInterface() {propertyValues {Curve {keys [
           {arriveTangent=0.0815, interpMode=RCIM_Cubic, leaveTangent=0.0815, tangentMode=RCTM_User, time=0, value=1}
           {arriveTangent=-2.3986, interpMode=RCIM_Cubic, leaveTangent=-2.3986, tangentMode=RCTM_User, time=1, value=0}
          ]}}}
       CurveIndex=$System.FN_Contrail_Setup.Dive Value
      )
     )
    )
   )
   &AddVelocity(
    Module.Use Falloff Distance=false
    Origin Offset=&AddVectorToPosition(
     Position=Linked() {linkedVariable {name=Particles.Offset, type=Vector3f}}
     Vector=&RotateVector(
      Vector To Rotate:Vector3f=&MergeHelper_VectorSwitch(
       0=<0 0 -5>
       1=&Lerp_Vector(
        A=<0 0 -200>
        B=&Lerp_Vector(A=<0 50 -200>, B=<5 50 -200>, Alpha=$System.BankAngle)
        Alpha=$System.FN_Contrail_Setup.Dive Value
       )
       2=&Lerp_Vector(
        A=<0 0 -250>
        B=&Lerp_Vector(A=<0 50 -200>, B=<5 50 -200>, Alpha=$System.BankAngle)
        Alpha=$System.FN_Contrail_Setup.Dive Value
       )
       3=<0 0 0>
       4=<0 0 0>
       Index=&SpawnGroup()
       Num=::NewEnumerator3
      )
      Yaw:NiagaraFloat=&Lerp_Float(A=0, B=&Lerp_Float(A=1, B=-1, Alpha=$System.BankAngle), Alpha=$System.FN_Contrail_Setup.Dive Value)
      Pitch:NiagaraFloat=0
      Roll:NiagaraFloat=&Lerp_Float(
       A=&FN_FrontEndFloatModifier(Gameplay=40, Front End=20)
       B=&FN_FrontEndFloatModifier(Gameplay=20, Front End=6)
       Alpha=&FloatFromCurve(
        FloatCurve=DataInterface() {propertyValues {Curve {keys [
            {arriveTangent=0.0815, interpMode=RCIM_Cubic, leaveTangent=0.0815, tangentMode=RCTM_User, time=0, value=1}
            {arriveTangent=-2.3986, interpMode=RCIM_Cubic, leaveTangent=-2.3986, tangentMode=RCTM_User, time=1, value=0}
           ]}}}
        CurveIndex=$System.FN_Contrail_Setup.Dive Value
       )
      )
      Rotation Rate:NiagaraFloat=$Engine.DeltaTime
     )
    )
    Override Seed=false
    Velocity Mode=::NewEnumerator1
    Velocity Speed=&MergeHelper_FloatSwitch(
     0=20
     1=&FN_FrontEndFloatModifier(
      Gameplay=&RandomRangeFloat(Minimum=5, Maximum=10)
      Front End=&RandomRangeFloat(Minimum=75, Maximum=150)
     )
     2=&Lerp_Float(
      A=&RandomRangeFloat(Minimum=500, Maximum=800)
      B=&RandomRangeFloat(Minimum=600, Maximum=1000)
      Alpha=$System.FN_Contrail_Setup.Dive Value
     )
     3=?
     Index=&SpawnGroup()
     Num=::NewEnumerator1
    )
   ) {module=AddVelocity002}
   &SubUVAnimation(End Frame=3, SubUV Animation Mode:ENiagara_SubUVLookupMode=::NewEnumerator2)
  ]
  particleUpdateScript [
   &ParticleState()
   &GravityForce(Gravity=&MergeHelper_VectorSwitch(0=<0 0 -10>, 1=<0 0 -980>, 2=<0 0 -3000>, 3=?, Index=&SpawnGroup(), Num=::NewEnumerator1))
   &Drag(Use Linear Drag=true, Use Rotational Drag=true)
   &ScaleColor(
    Scale Alpha=&MergeHelper_FloatSwitch(
     0=&FloatFromCurve(
      FloatCurve=DataInterface() {propertyValues {Curve {keys [
          {arriveTangent=9.8898, interpMode=RCIM_Cubic, leaveTangent=9.8898, tangentMode=RCTM_User, time=0, value=0}
          {interpMode=RCIM_Cubic, tangentMode=RCTM_User, time=0.3, value=1}
          {arriveTangent=-3.6098, tangentMode=RCTM_User, time=1, value=0}
         ]}}}
      Scale Curve=2
     )
     1=&FloatFromCurve(
      FloatCurve=DataInterface() {propertyValues {Curve {keys [
          {interpMode=RCIM_Cubic, tangentMode=RCTM_User, time=0, value=0}
          {interpMode=RCIM_Cubic, tangentMode=RCTM_User, time=0.1, value=1}
          {interpMode=RCIM_Cubic, tangentMode=RCTM_User, time=1, value=0}
         ]}}}
      Scale Curve=&FN_FrontEndFloatModifier(
       Gameplay=&FN_Select_Float_from_Time_Of_Day(Morning=0.8, Day=1, Evening=0.8, Night=0.6)
       Front End=0.5
      )
     )
     2=&FloatFromCurve(
      FloatCurve=DataInterface() {propertyValues {Curve {keys [
          {interpMode=RCIM_Cubic, tangentMode=RCTM_User, time=0, value=0}
          {interpMode=RCIM_Cubic, tangentMode=RCTM_User, time=0.1, value=1}
          {interpMode=RCIM_Cubic, tangentMode=RCTM_User, time=1, value=0}
         ]}}}
      Scale Curve=&FN_Select_Float_from_Time_Of_Day(Morning=0.175, Day=0.2, Evening=0.175, Night=0.15)
     )
     3=?
     Index=&SpawnGroup()
     Num=::NewEnumerator1
    )
    Scale RGB=&MergeHelper_VectorSwitch(
     0=<1 1 1>
     1=&MakeVectorFromLinearColorRGB(Color=&ColorFromCurve(ColorCurve=DataInterface() {propertyValues {
         AlphaCurve {keys [<at 0 1 RCIM_Cubic>,<at 1 0 RCIM_Cubic>]}
         BlueCurve {keys [<at 0 1 RCIM_Cubic>,<at 0.5838 2 RCIM_Cubic>]}
         GreenCurve {keys [<at 0 0.6667 RCIM_Cubic>,<at 0.5838 1.8667 RCIM_Cubic>]}
         RedCurve {keys [<at 0 0 RCIM_Cubic>,<at 0.5838 1.6 RCIM_Cubic>]}
        }}))
     2=&MakeVectorFromLinearColorRGB(Color=&ColorFromCurve(ColorCurve=DataInterface() {propertyValues {
         AlphaCurve {keys [<at 0 1 RCIM_Cubic>,<at 0.6993 0.2168>,<at 1 0 RCIM_Cubic>]}
         BlueCurve {keys [<at 0.1395 1 RCIM_Cubic>,<at 0.7961 2 RCIM_Cubic>]}
         GreenCurve {keys [<at 0.1395 0.265 RCIM_Cubic>,<at 0.7961 1.6 RCIM_Cubic>]}
         RedCurve {keys [<at 0.1395 0.02 RCIM_Cubic>,<at 0.7961 1.4 RCIM_Cubic>]}
        }}))
     3=?
     Index=&SpawnGroup()
     Num=::NewEnumerator1
    )
    ScaleA=true
    ScaleRGB=true
   )
   &InterpolateOverTime(
    Rate Of Change=10
    Target Vector Value:Vector3f=&AddVectorToPosition(
     Position=$Particles.Position
     Vector=&RotateVectorByQuaternion(Vector To Rotate=<0 -200 0>, Quaternion=$Particles.SkeletalMeshLocation.SampledRotation)
    )
    Variable Type=::NewEnumerator2
   )
   &DynamicMaterialParameters(
    Index 0 Param 1=&MergeHelper_FloatSwitch(
     0=&RandomRangeFloat(Minimum=0, Maximum=1)
     1=&FloatFromCurve(FloatCurve=DataInterface() {propertyValues {Curve {keys [
          {arriveTangent=-0.013, interpMode=RCIM_Cubic, leaveTangent=-0.013, tangentMode=RCTM_User, time=0, value=1}
          {arriveTangent=-0.3234, interpMode=RCIM_Cubic, leaveTangent=-0.3234, tangentMode=RCTM_User, time=1.0011, value=0}
         ]}}})
     2=0
     3=0
     4=0
     Index=&SpawnGroup()
     Num=::NewEnumerator3
    )
    Index 0 Param 3=&FloatFromCurve(
     FloatCurve=DataInterface() {propertyValues {Curve {keys [{interpMode=RCIM_Cubic, tangentMode=RCTM_User, time=0, value=0.5}, {interpMode=RCIM_Cubic, tangentMode=RCTM_User, time=1, value=4}]}}}
     Scale Curve=&RandomRangeFloat(Minimum=4, Maximum=6)
    )
    Index 0 Param 4=&FloatFromCurve(
     FloatCurve=DataInterface() {propertyValues {Curve {keys [
         {arriveTangent=-1.9474, interpMode=RCIM_Cubic, leaveTangent=-1.9474, tangentMode=RCTM_User, time=0, value=1}
         {interpMode=RCIM_Cubic, tangentMode=RCTM_User, time=1, value=0}
        ]}}}
     Scale Curve=0.25
    )
    Module.Param0WriteEnabled=true
    Module.Param1WriteEnabled=false
    Module.Param2WriteEnabled=true
    Module.Param3WriteEnabled=true
   ) {module=DynamicMaterialParameters001}
   &ScaleMeshSize(Scale Factor=&VectorFromCurve(
     VectorCurve=DataInterface() {propertyValues {
       XCurve {keys [
         {arriveTangent=4.3367, interpMode=RCIM_Cubic, leaveTangent=4.3367, tangentMode=RCTM_User, time=0, value=1}
         {arriveTangent=0.25, interpMode=RCIM_Cubic, leaveTangent=0.25, time=0.4, value=3}
         <at 1 1.25>
        ]}
       YCurve {keys [
         {arriveTangent=1.6312, interpMode=RCIM_Cubic, leaveTangent=1.6312, tangentMode=RCTM_User, time=0, value=1}
         <at 0.3 1.2 RCIM_Cubic>
         <at 1 1>
        ]}
       ZCurve {keys [
         {arriveTangent=1.2374, interpMode=RCIM_Cubic, leaveTangent=1.2374, tangentMode=RCTM_User, time=0, value=0.9}
         <at 0.3 2 RCIM_Cubic>
         <at 1 0.9 RCIM_Cubic>
        ]}
      }}
     Scale Curve=<0.5 1.5 0.6>
    ))
   &SolveForcesAndVelocity(Clamp Velocity=false, Limit Acceleration=false)
   &ScaleSpriteSize(Scale Factor=&Vector2DFromFloat(Value=&FloatFromCurve(
      FloatCurve=DataInterface() {propertyValues {Curve {keys [
          {arriveTangent=4.807, interpMode=RCIM_Cubic, leaveTangent=4.807, tangentMode=RCTM_User, time=0, value=0.7}
          {arriveTangent=0.8929, interpMode=RCIM_Cubic, leaveTangent=0.8929, tangentMode=RCTM_User, time=1, value=4}
         ]}}}
      Scale Curve=&FN_FrontEndFloatModifier(Gameplay=0.6, Front End=0.4)
     )))
   /SetVariables(Particles.VisibilityTag=&SpawnGroup()) {isSetParameters=true}
   /SetVariables(Particles.SpikesColor:LinearColor=&MultiplyLinearColor(
     Linear Color=$Particles.Color
     Scale Factor=&MakeVector4FromVectorAndFloat(
      "Vector (RGB)"=<1 1 1>
      "Float (Alpha)"=&FloatFromCurve(
       FloatCurve=DataInterface() {propertyValues {Curve {keys [{interpMode=RCIM_Cubic, tangentMode=RCTM_User, time=0, value=1}, {interpMode=RCIM_Cubic, tangentMode=RCTM_User, time=0.5, value=0}]}}}
       CurveIndex=$System.FN_Contrail_Setup.Dive Value
      )
     )
    )) {isSetParameters=true}
  ]
  rendererClasses [&NiagaraMeshRendererProperties, &NiagaraSpriteRendererProperties]
  renderers [
   {rendererClass=&NiagaraMeshRendererProperties, rendererIndex=0}
   {rendererClass=&NiagaraSpriteRendererProperties, rendererIndex=1}
   {rendererClass=&NiagaraSpriteRendererProperties, rendererIndex=2}
  ]
 }
 WristRope {
  emitterUpdateScript [
   &EmitterState(Life Cycle Mode=::NewEnumerator1, Loop Behavior=::NewEnumerator1, Loop Duration Mode=::NewEnumerator1, UseLoopDelay=false)
   &SpawnBurst_Instantaneous(Spawn Count=2, Use Spawn Probability=false, UseLoopCountLimit=false)
  ]
  particleSpawnScript [
   &PartitionParticles()
   &InitializeParticle(
    Color=&MakeLinearColorFromVectorAndFloat(
     "Vector (RGB)"=&NormalizeVector(Vector To Normalize=$Engine.Owner.Velocity, Safe Normalize=true, Safe Fallback Threshold=0, Safe Fallback Vector=<0 0 0.001>)
     "Float (Alpha)"=1
    )
    Lifetime=5
    Mesh Scale Mode=::NewEnumerator3
    Mesh Uniform Scale=1.33
    UsePositionOffset=false
    Write Material Random=true
   )
   &InitialMeshOrientation(
    Mesh Axis To Orient=<0 1 0>
    Mesh Orientation Mode=::NewEnumerator2
    Rotation=&SelectVectorFromArray(
     Direct Array Index=$Particles.PartitionParticles.Partition
     Vector Selection Array=DataInterface() {propertyValues {FloatData [<0 0.1982 0>, <0 -0.6486 0>]}}
    )
    Use Rotation Vector=true
   )
  ]
  particleUpdateScript [
   &ParticleState(Kill Particles When Lifetime Has Elapsed=false, Let Infinitely Lived Particles Die When Emitter Deactivates=true)
   &DynamicMaterialParameters(
    Index 0 Param 2=&Lerp_Float(A=0.5, B=1, Alpha=$System.Timeline.Float)
    Index 0 Param 3=4
    Index 1 Param 2=$Engine.RealTime
    Module.Index0Param0Write=false
    Module.Index0Param1Write=true
    Module.Index0Param2Write=true
    Module.Index0Param3Write=false
    Module.Index1Param0Write=true
    Module.Index1Param1Write=true
    Module.Index1Param2Write=true
    Module.Index1Param3Write=true
    Write Parameter Index 1=true
   )
   &SkeletalMeshLocation(
    Bone / Socket Index=&SelectIntFromArray(Direct Array Index=$Particles.PartitionParticles.Partition, Int Selection Array=DataInterface() {propertyValues {IntData [0,1]}})
    MaskSpawnGroup=false
    Mesh Sampling Type=::NewEnumerator5
    Sampled Orientation Coordinate Space=::Local
    Sampled Position Offset Coordinate Space:ENiagaraSkelMeshTransforms=::NewEnumerator2
    Skeletal Mesh=$System.Skeletal Mesh
    Socket Sampling Mode=::NewEnumerator5
   )
   /SetVariables(Particles.VisibilityTag=&SelectIntFromArray(Direct Array Index=$Particles.PartitionParticles.Partition, Int Selection Array=DataInterface() {propertyValues {IntData [0,1]}})) {isSetParameters=true}
  ]
  rendererClasses [&NiagaraMeshRendererProperties]
  renderers [
   {rendererClass=&NiagaraMeshRendererProperties, rendererIndex=0}
   {rendererClass=&NiagaraMeshRendererProperties, rendererIndex=1}
  ]
 }
}
summary {
 systemName=NS_Contrail_AquaPeony
 emitters [
  {emitterName=WristRope, rendererClasses [&NiagaraMeshRendererProperties]}
  {emitterName=WaterParticles, rendererClasses [&NiagaraMeshRendererProperties, &NiagaraSpriteRendererProperties]}
  {emitterName=RopeRibbon, rendererClasses [&NiagaraRibbonRendererProperties]}
  {emitterName=Buoy, rendererClasses [&NiagaraMeshRendererProperties]}
  {emitterName=RopeKnot, rendererClasses [&NiagaraMeshRendererProperties]}
  {emitterName=Shark, rendererClasses [&NiagaraMeshRendererProperties]}
 ]
 userVariables [
  User.bIsFrontEnd {type=NiagaraBool}
  User.bIsFrontEndPreview {type=NiagaraBool}
  User.Debug {type=NiagaraBool}
  User.DebugVelocity {type=Vector3f, description="", defaultValue {struct=Vector3f, value=<0 0 1>}}
  User.ParaGlideLeanAlpha {type=NiagaraFloat, defaultValue=0, description=""}
  User.PlayerSpeed {type=NiagaraFloat, defaultValue=0, description=""}
  User.RibbonIDOffset {type=NiagaraInt32}
 ]
}
